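; Construction script for the industrial mine cart C1 building prop that tints its meshes from its building texture data.
; It fetches the actor's texture data array and checks that the selected entry is valid.
; It then loops over every static mesh component, passing that entry's tint, converted to a linear color, into SetColoronSecondaryMesh.
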(fn ConstructionScript
  (bind _outtexturedataarray (Building|GetBuildingTextureData))
  (if (Utilities|IsValid (Utilities|Array|Get(acopy) _outtexturedataarray) :!class /Script/Engine.KismetSystemLibrary)
    (for _array_element (Actor|GetComponentsByClass "/Script/Engine.StaticMeshComponent")
      (Default|SetColoronSecondaryMesh (Math|Conversions|ToLinearColor(Color) (Variables|/Script/FortniteGame.BuildingTextureData|GetTintColor (Utilities|Array|Get(acopy) _outtexturedataarray))) _array_element))))
